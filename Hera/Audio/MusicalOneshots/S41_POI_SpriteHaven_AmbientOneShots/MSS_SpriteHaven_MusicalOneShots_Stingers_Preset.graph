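# MetaSound Source preset that supplies the ambient musical one-shot stingers for the Sprite Haven POI.
# It inherits its graph from MSS_MusicEncounter_RandomStingerFromArray, which by its name picks a random stinger from an array, here filled with nine Sprite Haven wave assets.
# Inputs expose master volume plus a music-encounter pause flag and fade-out time.
# Output is one-shot stereo with a finished trigger, assigned to the Fort_Music_NPC sound class.
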MetaSoundSource /Hera/Audio/MusicalOneshots/S41_POI_SpriteHaven_AmbientOneShots/MSS_SpriteHaven_MusicalOneShots_Stingers_Preset
preset_parent=/MusicEncounter/MetaSounds/MSS_MusicEncounter_RandomStingerFromArray
inputs [
 MasterVolume {type=Float}
 Music.WaveAsset {type=WaveAsset:Array}
 MusicEncounter.bIsPaused {type=Bool}
 MusicEncounter.PauseFadeOutTime {type=Time}
 UE.Source.OnPlay {type=Trigger}
]
interfaces [UE.OutputFormat.Stereo:1.0, UE.Source.OneShot:1.0, UE.Source:1.1]
outputs [
 "UE.OutputFormat.Stereo.Audio:0" {type=Audio}
 "UE.OutputFormat.Stereo.Audio:1" {type=Audio}
 UE.Source.OneShot.OnFinished {type=Trigger}
]
pages [Default {input_defaults {
   MasterVolume=1
   MusicEncounter.PauseFadeOutTime=3
   MusicEncounter.bIsPaused=null
   UE.Source.OnPlay=null
   Music.WaveAsset [
    &/MX_S41_POI_SpriteHaven_AmbientOneShot_STNG_9
    &/MX_S41_POI_SpriteHaven_AmbientOneShot_STNG_8
    &/MX_S41_POI_SpriteHaven_AmbientOneShot_STNG_7
    &/MX_S41_POI_SpriteHaven_AmbientOneShot_STNG_6
    &/MX_S41_POI_SpriteHaven_AmbientOneShot_STNG_5
    &/MX_S41_POI_SpriteHaven_AmbientOneShot_STNG_4
    &/MX_S41_POI_SpriteHaven_AmbientOneShot_STNG_3
    &/MX_S41_POI_SpriteHaven_AmbientOneShot_STNG_2
    &/MX_S41_POI_SpriteHaven_AmbientOneShot_STNG_1
   ]
  }}]
properties {
 assetImportData=&self:AssetImportData
 numChannels=2
 outputFormat=Stereo
 sampleRate=48000
 soundClassObject=&Fort_Music_NPC
}
